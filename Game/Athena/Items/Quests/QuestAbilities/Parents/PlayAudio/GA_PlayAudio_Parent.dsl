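; Parent quest ability that plays a sequence of voice-over sounds, plus a looping sound, on the quest's effect causer or the avatar.
; On authority, a stacking IsPlaying effect serves as a queue, watched for stack drops, before net sync and an optional delay.
; Finished-audio callbacks and timers advance the sequence, sending audio to an envelope-follower bus for ducking.
; A hotfix check or invalid sound ends the ability, which removes its effects and looping audio.
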(event Ability|EventFailedToActivatePassiveAbility
  (params (ActorInfo :type Gameplay%20Ability%20Actor%20Info%20Structure))
  (|CallParentFunction ActorInfo))

(event Ability|EventActivateAbility
  (bind _playdelay (Variables|Setup|GetPlayDelay))
  (bind _placeinline (Variables|Data|GetPlaceInLine))
  (|CallParentFunction)
  (Ability|HasAuthority
    (:True
      (bind _returnvalue (Ability|ApplyGameplayEffectToOwner "/Game/Athena/Items/Quests/QuestAbilities/Parents/PlayAudio/GE_PlayAudio_IsPlaying.GE_PlayAudio_IsPlaying_C"))
      (bind _output_get (Variables|Data|SetGEVOPlayingHandle _returnvalue))
      (bind _returnvalue_2 (Ability|GameplayEffect|GetActiveGameplayEffectStackCount _output_get))
      (bind _output_get_2 (Variables|Data|SetPlaceInLine _returnvalue_2))
      (if (> _output_get_2 1)
        (bind (_asynctaskproxy _handle _newcount _oldcount) (Ability|Tasks|WaitForGameplayEffectStackChange _output_get)
          (:OnChange
            (if (< _newcount _oldcount)
              (Math|Integer|DecrementInt _placeinline
                (:%20%20
                  (if (<= _placeinline 1)
                    (GameplayTasks|EndTask _asynctaskproxy)
                    (Utilities|FlowControl|DoOnce false
                      (:Completed
                        _))))))))
        (else
          _))
      (Ability|Tasks|WaitNetSync "BothWait"
        (:OnSync
          (if (<= _playdelay 0)
            (CallFunction|HandleVO)
            (else
              (Utilities|Time|SetTimerbyEvent (EventDispatchers|CreateEvent :!event Custom|HandleVO) _playdelay))))))
    (:False
      _)))

(event Custom|HandleVO
  (bind _returnvalue (Ability|GetGrantedbyEffectContext))
  (bind _returnvalue_2 (Ability|EffectContext|GetEffectCauser _returnvalue))
  (bind _loopingsound (Variables|Setup|GetLoopingSound))
  (if (FortQuestGlobals|IsNonBusrideChallengeVOHotfixedOff (Variables|Setup|GetAssetTags))
    (Ability|EndAbility)
    (else
      (Default|Play)
      (if (Variables|Setup|GetPlayOnQuestActor)
        (Utilities|IsValid _returnvalue_2
          (:Is%20Valid
            (bind _returnvalue_3 (Audio|SpawnSoundAttached :Sound _loopingsound :AttachToComponent (Variables|/Script/Engine.Actor|GetRootComponent _returnvalue_2) :AttachPointName "None" :LocationType "KeepRelativeOffset" :bStopWhenAttachedToDestroyed false :VolumeMultiplier 1.0 :PitchMultiplier 1.0 :StartTime 0.0 :bAutoDestroy false))
            (Variables|Data|SetLoopingSoundHandle _returnvalue_3))
          (:Is%20Not%20Valid
            _)))
      (bind _returnvalue_4 (Audio|SpawnSoundAttached :Sound _loopingsound :AttachToComponent (Variables|/Script/Engine.Actor|GetRootComponent (Ability|GetAvatarActorfromActorInfo)) :AttachPointName "None" :LocationType "KeepRelativeOffset" :bStopWhenAttachedToDestroyed false :VolumeMultiplier 1.0 :PitchMultiplier 1.0 :StartTime 0.0 :bAutoDestroy false))
      (Variables|Data|SetLoopingSoundHandle _returnvalue_4))))

(event Ability|EventOnEndAbility
  (params (bWasCancelled :type Boolean))
  (bind _loopingsoundhandle (Variables|Data|GetLoopingSoundHandle))
  (Ability|RemoveGameplayEffectFromOwnerWithAssetTags (Variables|Setup|GetAssetTags))
  (Ability|RemoveGameplayEffectFromOwnerWithHandle (Variables|Data|GetGEVOPlayingHandle) 1)
  (Audio|Components|Audio|Stop _loopingsoundhandle)
  (Components|DestroyComponent _loopingsoundhandle))

(fn AdvanceSequence
  (Math|Integer|IncrementInt (Variables|Data|GetCurrentIdx)
    (:%20%20
      (Default|PlayCurrentItem))))

(fn GetCurrentItem
  (bind _soundsequence (Variables|Setup|GetSoundSequence))
  (Utilities|IsValid _soundsequence
    (:Is%20Valid
      (bind _outsoundsequencedata (Class|FortSoundSequence|GetSoundSequenceDataforIndex _soundsequence (Variables|Data|GetCurrentIdx)))
      (return _outsoundsequencedata))
    (:Is%20Not%20Valid
      ;[1 1 1] SoundSequence can be invalid during development, so log a message and let the user decide if that's ok at this stage in development or not
      (Development|PrintString (Utilities|String|Append "SoundSequence was null for Quest GA with tags: " (GameplayTags|GetDebugStringfromGameplayTagContainer (Variables|Setup|GetAssetTags))) true true "(R=1.000000,G=0.799738,B=0.000000,A=1.000000)" 4.0)
      (return 0))))

(fn Play
  (Networking|CanExecuteCosmeticEvents
    (:True
      (Default|PlayCurrentItem))))

(fn PlayCurrentItem
  (bind _returnvalue (Ability|GetGrantedbyEffectContext))
  (bind _returnvalue_2 (Ability|EffectContext|GetEffectCauser _returnvalue))
  (bind _outputdelegate (EventDispatchers|CreateEvent :!event StartDelay))
  (bind _envfolloweraudiobus (Variables|Setup|GetEnvFollowerAudioBus))
  (bind _output (Default|GetCurrentItem))
  (bind _output_sound (Utilities|Struct|BreakSoundSequenceData _output))
  (Utilities|IsValid _output_sound
    (:Is%20Valid
      (if (Variables|Setup|GetPlayOnQuestActor)
        (Utilities|IsValid _returnvalue_2
          (:Is%20Valid
            (bind _returnvalue_3 (Audio|SpawnSoundAttached _output_sound (Variables|/Script/Engine.Actor|GetRootComponent _returnvalue_2)))
            (Audio|Components|Audio|SetAudioBusSendPostEffect _returnvalue_3 _envfolloweraudiobus 1.0 :!class /Script/Engine.AudioComponent) ; Envelope follower for modulation ducking
            (EventDispatchers|BindEventtoOnAudioFinished _returnvalue_3 _outputdelegate))
          (:Is%20Not%20Valid
            _)))
      (bind _returnvalue_4 (Audio|SpawnSoundAttached _output_sound (Variables|/Script/Engine.Actor|GetRootComponent (Ability|GetAvatarActorfromActorInfo))))
      (Audio|Components|Audio|SetAudioBusSendPostEffect _returnvalue_4 _envfolloweraudiobus 1.0 :!class /Script/Engine.AudioComponent) ; Envelope follower for modulation ducking
      (EventDispatchers|BindEventtoOnAudioFinished _returnvalue_4 _outputdelegate))
    (:Is%20Not%20Valid
      (Ability|EndAbility))))

(fn StartDelay
  (bind _output (Default|GetCurrentItem))
  (bind (_output_sound _output_delay) (Utilities|Struct|BreakSoundSequenceData _output))
  (if (== _output_delay 0.0)
    (Default|AdvanceSequence)
    (else
      (Utilities|Time|SetTimerbyEvent (EventDispatchers|CreateEvent :!event AdvanceSequence) _output_delay))))
